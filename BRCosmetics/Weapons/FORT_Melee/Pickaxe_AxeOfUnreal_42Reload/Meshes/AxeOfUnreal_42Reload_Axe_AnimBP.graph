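# Animation Blueprint for the AxeOfUnreal_42Reload pickaxe that makes four front and back "pt" bones float and jiggle on top of its input pose.
# Each update, real time drives sine and cosine offsets at staggered rates and phases.
# The AnimGraph additively offsets MSTR parent-named bones by fixed amounts, applies those oscillations, then runs root-relative spring AnimDynamics per bone.
# A second, larger-amplitude oscillation chain and a pawn-owner lookup are unconnected.
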AnimBlueprint /BRCosmetics/Weapons/FORT_Melee/Pickaxe_AxeOfUnreal_42Reload/Meshes/AxeOfUnreal_42Reload_Axe_AnimBP
target_skeleton=&/AxeOfUnreal_42Reload_Axe_Skeleton
graphs [
 AnimGraph {nodes [
   AnimGraphNode_SubInput_0 = Input Pose()
   n1 = "Transform (Modify) Bone - Bone: MSTR_pt_front_2"(ComponentPose=Local To Component(LocalPose=@AnimGraphNode_SubInput_0), Alpha=1, Translation="7.2,0,4", Rotation="0,0,0", Scale="1,1,1") {translationMode=BMM_Additive, translationSpace=BCS_BoneSpace, boneToModify {boneName=MSTR_pt_front_2}, translation {x=7.2, z=4}}
   n2 = "Transform (Modify) Bone - Bone: MSTR_pt_front_1"(ComponentPose=@n1, Alpha=1, Translation="3.2,0,4", Rotation="0,0,0", Scale="1,1,1") {translationMode=BMM_Additive, translationSpace=BCS_BoneSpace, boneToModify {boneName=MSTR_pt_front_1}, translation {x=3.2, z=4}}
   n3 = "Transform (Modify) Bone - Bone: MSTR_pt_bck_2"(ComponentPose=@n2, Alpha=1, Translation="-7.2,0,4", Rotation="0,0,0", Scale="1,1,1") {translationMode=BMM_Additive, translationSpace=BCS_BoneSpace, boneToModify {boneName=MSTR_pt_bck_2}, translation {x=-7.2, z=4}}
   n4 = "Transform (Modify) Bone - Bone: MSTR_pt_bck_1"(ComponentPose=@n3, Alpha=1, Translation="-3.2,0,4", Rotation="0,0,0", Scale="1,1,1") {translationMode=BMM_Additive, translationSpace=BCS_BoneSpace, boneToModify {boneName=MSTR_pt_bck_1}, translation {x=-3.2, z=4}}
   n5 = "Transform (Modify) Bone - Bone: pt_front_2"(ComponentPose=@n4, Alpha=1, Translation=Get pt_fnt_2() -> (pt_fnt_2:Vector), Rotation="0,0,0", Scale="1,1,1") {
    rotationMode=BMM_Additive
    rotationSpace=BCS_BoneSpace
    translationMode=BMM_Additive
    translationSpace=BCS_BoneSpace
    boneToModify {boneName=pt_front_2}
   }
   n6 = "Transform (Modify) Bone - Bone: pt_front_1"(ComponentPose=@n5, Alpha=1, Translation=Get pt_fnt_1() -> (pt_fnt_1:Vector), Rotation="0,0,0", Scale="1,1,1") {translationMode=BMM_Additive, translationSpace=BCS_BoneSpace, boneToModify {boneName=pt_front_1}}
   n7 = Get pt_bck_2() -> (pt_bck_2:Vector)
   n8 = Anim Dynamics - Bone: pt_bck_2(
    ComponentPose="Transform (Modify) Bone - Bone: pt_bck_1"(
     ComponentPose="Transform (Modify) Bone - Bone: pt_bck_2"(ComponentPose=@n6, Alpha=1, Translation=@n7, Rotation="0,0,0", Scale="1,1,1") {translationMode=BMM_Additive, translationSpace=BCS_BoneSpace, boneToModify {boneName=pt_bck_2}}
     Alpha=1
     Translation=Get pt_bck_1() -> (pt_bck_1:Vector)
     Rotation="0,0,0"
     Scale="1,1,1"
    ) {translationMode=BMM_Additive, translationSpace=BCS_BoneSpace, boneToModify {boneName=pt_bck_1}}
    Alpha=0.8
    ExternalForce="0,0,0"
   ) {
    angularDampingOverride=1
    angularSpringConstant=100
    bLinearSpring=true
    bOverrideLinearDamping=true
    bUseGravityOverride=true
    linearDampingOverride=1
    linearSpringConstant=100
    simulationSpace=RootRelative
    boundBone {boneName=pt_bck_2}
    componentAppliedLinearAccClamp=<3000 3000 3000>
    componentLinearAccScale=<0.24 0.18 0.21>
    componentLinearVelScale=<0.3 0.3 0.25>
    gravityOverride {z=-600}
    physicsBodyDefinitions [{
      boundBone {boneName=pt_bck_2}
      boxExtents=<10 10 10>
      constraintSetup {angularLimitsMax=<0 0 0>, angularLimitsMin=<0 0 0>, angularTarget=<1 0 0>, linearAxesMax=<0 15 15>, linearAxesMin=<0 -15 -10>}
      localJointOffset=<0.8 0 0>
     }]
    relativeSpaceBone {boneName=b_AR_Root}
    simSpaceSettings {maxAngularAcceleration=40, maxAngularVelocity=8}
   }
   n9 = Anim Dynamics - Bone: pt_bck_1(ComponentPose=@n8, Alpha=0.8, ExternalForce="0,0,0") {
    angularDampingOverride=1
    angularSpringConstant=100
    bLinearSpring=true
    bOverrideLinearDamping=true
    bUseGravityOverride=true
    linearDampingOverride=1
    linearSpringConstant=100
    simulationSpace=RootRelative
    boundBone {boneName=pt_bck_1}
    componentAppliedLinearAccClamp=<3000 3000 3000>
    componentLinearAccScale=<0.24 0.18 0.21>
    componentLinearVelScale=<0.3 0.3 0.25>
    gravityOverride {z=-600}
    physicsBodyDefinitions [{
      boundBone {boneName=pt_bck_1}
      boxExtents=<10 10 10>
      constraintSetup {angularLimitsMax=<0 0 0>, angularLimitsMin=<0 0 0>, angularTarget=<1 0 0>, linearAxesMax=<0 10 10>, linearAxesMin=<0 -10 -10>}
      localJointOffset=<0.9 0 0>
     }]
    relativeSpaceBone {boneName=b_AR_Root}
    simSpaceSettings {maxAngularAcceleration=40, maxAngularVelocity=8}
   }
   n10 = Anim Dynamics - Bone: pt_front_2(ComponentPose=@n9, Alpha=0.8, ExternalForce="0,0,0") {
    angularDampingOverride=1
    angularSpringConstant=100
    bLinearSpring=true
    bOverrideLinearDamping=true
    bUseGravityOverride=true
    linearDampingOverride=1
    linearSpringConstant=100
    simulationSpace=RootRelative
    boundBone {boneName=pt_front_2}
    componentAppliedLinearAccClamp=<3000 3000 3000>
    componentLinearAccScale=<0.24 0.18 0.21>
    componentLinearVelScale=<0.3 0.3 0.25>
    gravityOverride {z=-600}
    physicsBodyDefinitions [{
      boundBone {boneName=pt_front_2}
      boxExtents=<10 10 10>
      constraintSetup {angularLimitsMax=<0 0 0>, angularLimitsMin=<0 0 0>, angularTarget=<1 0 0>, linearAxesMax=<0 10 10>, linearAxesMin=<0 -10 -10>}
      localJointOffset=<1.1 0 0>
     }]
    relativeSpaceBone {boneName=b_AR_Root}
    simSpaceSettings {maxAngularAcceleration=40, maxAngularVelocity=8}
   }
   n11 = Anim Dynamics - Bone: pt_front_1(ComponentPose=@n10, Alpha=0.8, ExternalForce="0,0,0") {
    angularDampingOverride=1
    angularSpringConstant=100
    bLinearSpring=true
    bOverrideLinearDamping=true
    bUseGravityOverride=true
    linearDampingOverride=1
    linearSpringConstant=100
    simulationSpace=RootRelative
    boundBone {boneName=pt_front_1}
    componentAppliedLinearAccClamp=<3000 3000 3000>
    componentLinearAccScale=<0.24 0.18 0.21>
    componentLinearVelScale=<0.3 0.3 0.25>
    gravityOverride {z=-600}
    physicsBodyDefinitions [{
      boundBone {boneName=pt_front_1}
      boxExtents=<10 10 10>
      constraintSetup {angularLimitsMax=<0 0 0>, angularLimitsMin=<0 0 0>, angularTarget=<1 0 0>, linearAxesMax=<0 15 15>, linearAxesMin=<0 -15 -10>}
      localJointOffset=<1 0 0>
     }]
    relativeSpaceBone {boneName=b_AR_Root}
    simSpaceSettings {maxAngularAcceleration=40, maxAngularVelocity=8}
   }
   Output Pose(Result=Component To Local(ComponentPose=@n11)) {layerGroup=None}
  ]}
 EventGraph {nodes [
   Event BlueprintUpdateAnimation()
   n1 = GetRealTimeSeconds()
   n2 = float + float(A=float * float(A=@n1, B=1.9) {is=title K2Node_PromotableOperator}, B=8)
   n3 = "Cos (Radians)"!(A:double=@n2) -> (ReturnValue:double)
   n4 = "Sin (Radians)"(A=@n2)
   > Set pt_fnt_2(pt_fnt_2:Vector=MakeVector(X=float * float(A=@n4, B=1.8), Y=float * float(A=@n3, B=1.6), Z=float * float(A=@n3, B=1.6))) -> (Output_Get:Vector)
   n5 = float + float(A=float * float(A=@n1, B=1.8), B=6)
   n6 = "Sin (Radians)"(A=@n5)
   n7 = "Cos (Radians)"(A=@n5)
   > Set pt_fnt_1(pt_fnt_1:Vector=MakeVector(X=float * float(A=@n6, B=1.5), Y=float * float(A=@n6, B=1.3), Z=float * float(A=@n7, B=1.1))) -> (Output_Get:Vector)
   n8 = float + float(A=float * float(A=@n1, B=1.6), B=6)
   n9 = "Sin (Radians)"(A=@n8)
   n10 = "Cos (Radians)"(A=@n8)
   > Set pt_bck_2(pt_bck_2:Vector=MakeVector(X=float * float(A=@n9, B=1.8), Y=float * float(A=@n10, B=1.2), Z=float * float(A=@n10, B=1.7))) -> (Output_Get:Vector)
   n11 = float + float(A=float * float(A=@n1, B=1.4), B=5)
   n12 = "Sin (Radians)"(A=@n11)
   n13 = "Cos (Radians)"(A=@n11)
   > Set pt_bck_1(pt_bck_1:Vector=MakeVector(X=float * float(A=@n12, B=1.5), Y=float * float(A=@n12, B=1.6), Z=float * float(A=@n13, B=0.9))) -> (Output_Get:Vector)
   n14 = GetRealTimeSeconds()
   n15 = float + float(A=float * float(A=@n14, B=1.9), B=8)
   n16 = "Sin (Radians)"(A=@n15)
   n17 = "Cos (Radians)"(A=@n15)
   Set pt_fnt_2(pt_fnt_2=MakeVector(X=float * float(A=@n16, B=3.3), Y=float * float(A=@n17, B=4.3), Z=float * float(A=@n16, B=1.5)))
   n18 = float + float(A=float * float(A=@n14, B=1.8), B=6)
   n19 = "Cos (Radians)"(A=@n18)
   n20 = "Sin (Radians)"(A=@n18)
   > Set pt_fnt_1(pt_fnt_1=MakeVector(X=float * float(A=@n20, B=1.3), Y=float * float(A=@n20, B=2.3), Z=float * float(A=@n19, B=1)))
   n21 = float + float(A=float * float(A=@n14, B=1.6), B=6)
   n22 = "Cos (Radians)"(A=@n21)
   n23 = "Sin (Radians)"(A=@n21)
   > Set pt_bck_2(pt_bck_2=MakeVector(X=float * float(A=@n23, B=3.2), Y=float * float(A=@n22, B=4.2), Z=float * float(A=@n22, B=1.7)))
   n24 = float + float(A=float * float(A=@n14, B=1.4), B=5)
   n25 = "Sin (Radians)"(A=@n24)
   n26 = "Cos (Radians)"(A=@n24)
   > Set pt_bck_1(pt_bck_1=MakeVector(X=float * float(A=@n25, B=1.1), Y=float * float(A=@n25, B=2.1), Z=float * float(A=@n26, B=-0.6)))
   TryGetPawnOwner()
  ]}
]
variables [
 pt_bck_1 {type=Vector}
 pt_bck_2 {type=Vector}
 pt_fnt_1 {type=Vector}
 pt_fnt_2 {type=Vector}
]
